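# Unlit, two-sided translucent Niagara ribbon and mesh particle material for the Transition_IAS UI effect, producing a scanlined gradient mask.
# A gradient texture, sampled with swapped, controllable UVs, is shaped by a power, a side edge fade and particle alpha.
# That mask is multiplied by hard bands from a Ceil of a sine along the sampled UVs, biased by ScanOffset and a particle dynamic parameter.
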Material /Game/Effects/Fort_Effects/UI/Transition_IAS/Resources/M_Progress
graph {
 properties {
  bUsedWithNiagaraMeshParticles=true
  bUsedWithNiagaraRibbons=true
  blendMode=BLEND_Translucent
  pixelDepthOffsetMode=PDOM_Legacy
  shadingModel=MSM_Unlit
  twoSided=true
  usedShadingModels=Unlit
 }
 nodes [
  n1 = TextureObjectParameter() {
   group=Ribbon
   parameterName=Ribbon_Texture
   sortPriority=0
   texture=&/T_Vert_Gradient_Soft
  }
  n2 = VectorParameter() {group=Ribbon, parameterName=Ribbon_UVControl, channelNames {a=VPan, b=UPan, g=VScale, r=UScale}, defaultValue=<rgba 1 2 0 0.25>}
  n3 = TextureCoordinate()
  n4 = MaterialFunctionCall(
   Texture=@n1
   UVControl_RGB=@n2
   UVControl_A=@n2
   "[Override]UVInput"=MaterialFunctionCall(XY=@n3) {materialFunction=&Swizzle}.YX
  ) {materialFunction=&FXCommon_TextureSample}
  n5 = Sine(None=Multiply(
    A=MaterialFunctionCall(Float2=@n4.UV) {materialFunction=&BreakOutFloat2Components}.G
    B=ScalarParameter() {defaultValue=50, parameterName=ScanDensity}
   ))
  n6 = Add(A=DynamicParameter().Param1, B=ScalarParameter() {parameterName=ScanOffset})
  n7 = Power(Base=@n4.R, Exp=ScalarParameter() {defaultValue=0.7, parameterName=Power})
  n8 = VectorParameter() {group=SideFade, parameterName=SideFade_EdgeMask_Control, channelNames {a="", b=Power, g=Multiply, r=Offset}, defaultValue=<rgba 0 15 2 1>}
  n9 = MaterialFunctionCall!(
   UorV=MaterialFunctionCall(Float2=@n3) {materialFunction=&BreakOutFloat2Components}.G
   EdgeControl=@n8
  ) {materialFunction=/Game/Effects/Niagara/MasterMaterials/Functions/Common/FXCommon_EdgeMask_Simple}
  Multiply(A=Ceil(None=Add(A=@n5, B=@n6)), B=Multiply(A=Multiply(A=@n7, B=@n9.Mask), B=ParticleColor().A))
 ]
}
