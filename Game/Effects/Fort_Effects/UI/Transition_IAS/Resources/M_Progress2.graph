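# Translucent, unlit, two-sided ribbon material for the Transition_IAS UI effect, producing an animated scan-band mask.
# A gradient texture, sampled with swapped, panning UVs, is shaped by a power curve, a side edge mask and particle alpha.
# The same panned coordinate drives a sine, offset along V and sharpened into bands.
# These multiply together with V-based falloffs that fade both ends of the V range.
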Material /Game/Effects/Fort_Effects/UI/Transition_IAS/Resources/M_Progress2
graph {
 properties {
  bUsedWithNiagaraRibbons=true
  blendMode=BLEND_Translucent
  pixelDepthOffsetMode=PDOM_Legacy
  shadingModel=MSM_Unlit
  twoSided=true
  usedShadingModels=Unlit
 }
 nodes [
  n1 = MaterialFunctionCall(Float2=TextureCoordinate()) {materialFunction=&BreakOutFloat2Components}
  n2 = Multiply(A=OneMinus(None=@n1.G)) {constB=5}
  n3 = TextureObjectParameter() {
   group=Ribbon
   parameterName=Ribbon_Texture
   sortPriority=0
   texture=&/T_Vert_Gradient_Soft
  }
  n4 = VectorParameter() {group=Ribbon, parameterName=Ribbon_UVControl, channelNames {a=VPan, b=UPan, g=VScale, r=UScale}, defaultValue=<rgba 1 2 0.5 0>}
  n5 = TextureCoordinate()
  n6 = MaterialFunctionCall(
   Texture=@n3
   UVControl_RGB=@n4
   UVControl_A=@n4
   "[Override]UVInput"=MaterialFunctionCall(XY=@n5) {materialFunction=&Swizzle}.YX
  ) {materialFunction=&FXCommon_TextureSample}
  n7 = MaterialFunctionCall(Float2=@n6.UV) {materialFunction=&BreakOutFloat2Components}
  n8 = ScalarParameter() {defaultValue=10, parameterName=ScanDensity}
  n9 = ScalarParameter() {defaultValue=0.8, parameterName=ZDisPow}
  n10 = Add(
   A=Sine(None=Multiply(A=@n7.R, B=@n8))
   B=LinearInterpolate(B=ScalarParameter() {defaultValue=-0.75, parameterName=ScanOffset}, Alpha=Power(Base=@n1.G, Exp=@n9)) {constA=-1.1, constB=-1}
  )
  n11 = ScalarParameter() {defaultValue=15, parameterName=hardness}
  n12 = Saturate(None=Power(Base=Multiply(A=@n10, B=@n11)))
  n13 = ScalarParameter() {defaultValue=0.8, parameterName=Power}
  n14 = MaterialFunctionCall(Float2=@n5) {materialFunction=&BreakOutFloat2Components}
  n15 = VectorParameter() {group=SideFade, parameterName=SideFade_EdgeMask_Control, channelNames {a="", b=Power, g=Multiply, r=Offset}, defaultValue=<rgba 0 15 2 1>}
  n16 = Multiply(
   A=Multiply(
    A=Power(Base=@n6.R, Exp=@n13)
    B=MaterialFunctionCall!(UorV=@n14.G, EdgeControl=@n15) {materialFunction=/Game/Effects/Niagara/MasterMaterials/Functions/Common/FXCommon_EdgeMask_Simple}.Mask
   )
   B=ParticleColor().A
  )
  Multiply(A=Saturate(None=Power(Base=@n2)), B=Multiply(A=Multiply(A=@n12, B=@n16), B=Power(Base=@n1.G) {constExponent=1}))
 ]
}
