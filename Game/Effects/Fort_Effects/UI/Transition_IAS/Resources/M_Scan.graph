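# Translucent unlit Niagara mesh-particle material that draws a glowing cyan scan band, apparently for a UI transition.
# A soft vertical gradient, shifted by a dynamic VOffset, is warped by panning Perlin noise and V-scaled by a repeating U profile to shape its front.
# Color adds that band to two multiplied panning grid textures, tinted by particle color.
# Opacity combines band falloffs, inverse Fresnel, depth fade, V edge mask and particle alpha.
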Material /Game/Effects/Fort_Effects/UI/Transition_IAS/Resources/M_Scan
graph {
 properties {
  bUsedWithNiagaraMeshParticles=true
  blendMode=BLEND_Translucent
  pixelDepthOffsetMode=PDOM_Legacy
  shadingModel=MSM_Unlit
  usedShadingModels=Unlit
 }
 nodes [
  n1 = ParticleColor()
  n2 = Power(Base=DepthFade() {fadeDistanceDefault=50})
  n3 = Fresnel() {baseReflectFraction=0, exponent=1}
  n4 = TextureObjectParameter() {
   group=MainTexture
   parameterName=MainTexture_Texture
   sortPriority=0
   texture=/Game/Global/Textures/Default/Masks_and_Gradients/T_Vert_Gradient_Soft
  }
  n5 = VectorParameter() {group=MainTexture, parameterName=MainTexture_UVOffset, channelNames {a=0, b=0, g=VOffset, r=UOffset}, defaultValue=<rgba 0 -0.2936 0 0>}
  n6 = Add(A=@n5.VOffset, B=DynamicParameter() {defaultValue=<rgba 0 1 1 1>, paramNames [VOffset,Param2,Param3,Param4]}.VOffset)
  n7 = TextureObjectParameter() {
   group=NoiseA
   parameterName=NoiseA_Texture
   samplerType=SAMPLERTYPE_LinearGrayscale
   sortPriority=0
   texture=&T_SimplePerlin_Greyscale
  }
  n8 = VectorParameter() {group=NoiseA, parameterName=NoiseA_UVControl, channelNames {a=VPan, b=UPan, g=VScale, r=UScale}, defaultValue=<rgba 2 0.5 0 0.3>}
  n9 = VectorParameter() {
   group=NoiseA
   parameterName=NoiseA_NoiseControl
   sortPriority=60
   channelNames {a=VStrength, b=UStrength, g=VOffset, r=UOffset}
   defaultValue=<rgba -0.5 -0.5 0 0.5>
  }
  n10 = MaterialFunctionCall!(
   NoiseValue=MaterialFunctionCall(Texture=@n7, UVControl_RGB=@n8, UVControl_A=@n8) {materialFunction=&FXCommon_TextureSample}.RG
   Control_RGB=@n9
   Control_A=@n9
  ) {materialFunction=/Game/Effects/Niagara/MasterMaterials/Functions/Common/FXCommon_NoiseControl}
  n11 = Power(Base=MaterialFunctionCall(Float2=TextureCoordinate()) {materialFunction=&BreakOutFloat2Components}.G)
  n12 = VectorParameter() {group=MainTexture, parameterName=MainTexture_UVControl, channelNames {a=VPan, b=UPan, g=VScale, r=UScale}, defaultValue=<rgba 1 1 0 0>}
  n13 = ScalarParameter() {defaultValue=1.5, parameterName=V2}
  n14 = TextureCoordinate()
  n15 = Frac(None=Multiply(A=MaterialFunctionCall(Float2=@n14) {materialFunction=&BreakOutFloat2Components}.R) {constB=2})
  n16 = OneMinus(None=@n15)
  n17 = OneMinus(None=Multiply(A=Min(A=@n16, B=@n15)) {constB=2})
  n18 = Power(Base=OneMinus(None=Power(Base=@n17) {constExponent=0.5})) {constExponent=1}
  n19 = MaterialFunctionCall(Float2=TextureCoordinate()) {materialFunction=&BreakOutFloat2Components}
  n20 = MaterialFunctionCall!(
   Texture=@n4
   AdditionalUVOffset=Add(A=AppendVector(A=@n5.UOffset, B=@n6), B=Multiply(A=@n10.NoiseValue, B=@n11))
   UVControl_RGB=MaterialFunctionCall(R=@n12.UScale, G=LinearInterpolate(A=@n12.VScale, B=@n13, Alpha=@n18), B=@n12.UScale) {materialFunction=&AppendMany}.RGB
   UVControl_A=@n12.VPan
   "[Override]TimeInput"
   "[Override]ScaleUVByCenter"=StaticBool() {value=true}
   "[Override]UVInput"=MaterialFunctionCall(X=@n19.R, Y=OneMinus(None=@n19.G)) {materialFunction=&MakeFloat2}.Result
  ) {materialFunction=/Game/Effects/Niagara/MasterMaterials/Functions/Common/FXCommon_TextureSample_V2}
  n21 = VectorParameter() {parameterName=Ref, defaultValue=<rgba 15 1.2 2.5 0.05>}
  n22 = Power(Base=@n20.R, Exp=@n21.R)
  n23 = Power(Base=@n20.R, Exp=@n21.G)
  n24 = Multiply(
   A=LinearInterpolate(A=Multiply(A=@n22, B=@n21.B), B=Multiply(A=@n23, B=@n21.A), Alpha=ScalarParameter() {defaultValue=0.5, parameterName=RefL})
   B=Power(Base=OneMinus(None=@n3)) {constExponent=10}
  )
  n25 = MaterialFunctionCall(Float2=TextureCoordinate()) {materialFunction=&BreakOutFloat2Components}
  n26 = VectorParameter() {
   group=EdgeMask
   parameterName=EdgeMask_EdgeMask_InvertControl
   channelNames {a="", b=Power, g=Multiply, r=Offset}
   defaultValue=<rgba 0 3 1 1>
  }
  n27 = VectorParameter() {group=EdgeMask, parameterName=EdgeMask_EdgeMask_Control, channelNames {a="", b=Power, g=Multiply, r=Offset}, defaultValue=<rgba 0 3 1 1>}
  Multiply(
   A=@n1.A
   B=Multiply(
    A=Multiply(A=@n2, B=@n24)
    B=MaterialFunctionCall!(UorV=@n25.G, Invert_EdgeControl=@n26, NoInvert_EdgeControl=@n27) {materialFunction=/Game/Effects/Niagara/MasterMaterials/Functions/Common/FXCommon_EdgeMask_Ex}.Mask
   )
  )
  n28 = Power(Base=@n20.R, Exp=ScalarParameter() {defaultValue=3, parameterName=FAF})
  n29 = TextureObjectParameter() {
   group=Texture
   parameterName=Texture_Texture
   samplerType=SAMPLERTYPE_LinearColor
   sortPriority=0
   texture=/Game/Packages/Fortress_SharedMaterials/Textures/T_Slowfield_GridSquares_01
  }
  n30 = VectorParameter() {group=Texture, parameterName=Texture_UVControl, channelNames {a=VPan, b=UPan, g=VScale, r=UScale}, defaultValue=<rgba 3 2 0 0.2>}
  n31 = TextureObjectParameter() {
   group=Texture
   parameterName=Texture2_Texture
   samplerType=SAMPLERTYPE_LinearColor
   sortPriority=0
   texture=&T_Slowfield_GridSquares_01
  }
  n32 = VectorParameter() {group=Texture, parameterName=Texture2_UVControl, channelNames {a=VPan, b=UPan, g=VScale, r=UScale}, defaultValue=<rgba 2 1.5 0 0.3>}
  n33 = Multiply(
   A=Power(Base=MaterialFunctionCall(Texture=@n29, UVControl_RGB=@n30, UVControl_A=@n30) {materialFunction=&FXCommon_TextureSample}.R) {constExponent=1}
   B=Power(Base=MaterialFunctionCall(Texture=@n31, UVControl_RGB=@n32, UVControl_A=@n32) {materialFunction=&FXCommon_TextureSample}.R) {constExponent=1}
  )
  n34 = VectorParameter() {group=RtC, parameterName=RtC_G2C_ColorA, defaultValue=<rgba 0 1.3333 2 1>}
  n35 = ScalarParameter() {defaultValue=15, parameterName=RtC2}
  Multiply(A=Multiply(A=Add(A=@n28, B=@n33), B=Multiply(A=@n34.RGB, B=@n35)), B=@n1.RGB)
 ]
}
